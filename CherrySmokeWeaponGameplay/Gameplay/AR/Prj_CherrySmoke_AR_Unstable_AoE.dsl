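; CherrySmoke AR projectile that can trigger an unstable area explosion when it stops.
; On stop, the server asks the owning DMR Blueprint whether the bullet is unstable, then runs an async targeting preset and applies the AoE gameplay effect to results.
; Caching the hit location drives an OnRep that spawns Niagara, sound and an optional scorch decal on remotes.
; FindTargetsAtLocation is uncalled here, and the explosion getter ignores its water input.
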(event EventBeginPlay)

(event Collision|EventActorBeginOverlap
  (params (OtherActor :type Actor%20Object%20Reference)))

(event EventTick
  (params (DeltaSeconds :type Float%20%28single-precision%29)))

(event EventOnStop
  (params (Hit_bBlockingHit :type Boolean) (Hit_bInitialOverlap :type Boolean) (Hit_Time :type Float%20%28single-precision%29) (Hit_Distance :type Float%20%28single-precision%29) (Hit_Location :type Vector :default "0, 0, 0") (Hit_ImpactPoint :type Vector :default "0, 0, 0") (Hit_Normal :type Vector :default "0, 0, 0") (Hit_ImpactNormal :type Vector :default "0, 0, 0") (Hit_PhysMat :type Physical%20Material%20Object%20Reference) (Hit_HitActor :type Actor%20Object%20Reference) (Hit_HitComponent :type Primitive%20Component%20Object%20Reference) (Hit_HitBoneName :type Name :default "None") (Hit_BoneName :type Name :default "None") (Hit_HitItem :type Integer) (Hit_ElementIndex :type Integer) (Hit_FaceIndex :type Integer) (Hit_TraceStart :type Vector :default "0, 0, 0") (Hit_TraceEnd :type Vector :default "0, 0, 0"))
  (bind _returnvalue (Actor|GetOwner))
  (Networking|SwitchHasAuthority
    (:Authority
      (bind _asb_cherry_smoke_dmr_athena (Utilities|Casting|CastToB_CherrySmoke_DMR_Athena _returnvalue))
      (Variables|Transient|SetOwningPlayerPawn (Actor|GetOwner _returnvalue))
      (if (Class|BCherrySmokeDMRAthena|ShouldFireUnstableBullet _asb_cherry_smoke_dmr_athena (Variables|Config|GetbIsFirstShotUnstable))
        (bind (_asynctaskproxy _targetinghandle) (Class|AsyncActionPerformTargeting|PerformTargetingAsyncAction 0 "/CherrySmokeWeaponGameplay/Gameplay/AR/Explosion/TP_CherrySmoke_Unstable_Explosion.TP_CherrySmoke_Unstable_Explosion" true)
          (:then
            (Development|PrintString "PROCCING UNSTABLE AOE !!!")
            (Variables|Transient|SetCachedHitLocation Hit_Location))
          (:Targeted
            (Ability|Tasks|ApplyGameplayEffectToHitResultsAsync (Ability|GetAbilitySystemComponent (Variables|Transient|GetOwningPlayerPawn)) (TargetingSystem|TargetingResults|GetTargetingResults (GameInstanceSubsystems|GetTargetingSubsystem) _targetinghandle) (Variables|Config|GetAoEGE) 0 1 (Class|AthenaFunctionLibrary|MakeSimpleFortEffectContext (Variables|Transient|GetOwningPlayerPawn) self) (Variables|Config|GetMax%20Targets%20to%20Process%20Per%20Tick%20CVar))))))))

(fn FindTargetsAtLocation
  (params (LocationToTarget :type Vector))
  (bind _outactors (Collision|SphereOverlapActors LocationToTarget (ScalableFloat|GetValueatLevel (Variables|Config|GetAoESize)) (Utilities|Array|MakeArray "ObjectTypeQuery3" "ObjectTypeQuery7") 0 (Utilities|Array|MakeArray (Variables|Transient|GetOwningPlayerPawn))))
  (return _outactors))

(fn GetExplosionAOE
  (params (IsInWater :type Boolean))
  (return (Variables|Config|GetAoENiagara)))

(fn OnRep_CachedHitLocation
  (bind _cachedhitlocation (Variables|Transient|GetCachedHitLocation))
  (Networking|SwitchHasAuthority
    (:Remote
      (Niagara|SpawnSystematLocation (CallFunction|GetExplosionAOE (Water|IsinWater (Variables|Water|GetWaterInteractionComponent))) _cachedhitlocation)
      (Audio|SpawnSoundatLocation (Variables|Sounds|GetCueExplosionSound) _cachedhitlocation)
      (if (Variables|Config|FX|GetUseAOEDecal)
        (CallFunction|SpawnAOEDecal _cachedhitlocation)))))

(fn SpawnAOEDecal
  (params (Location :type Vector))
  ;[0.15 0.0266 9.8e-05 0.5] Decal
  (Utilities|FlowControl|Sequence
    (:then_0
      (bind _outhit (Collision|SphereTraceByChannel (Transformation|GetActorLocation) (- (Transformation|GetActorLocation) "0.000000,0.000000,10.000000") (ScalableFloat|GetValueatLevel (Variables|Config|GetAoESize)) "TraceTypeQuery10"))
      (LocalVariables|SetImpactNrm (.ImpactNormal (Collision|BreakHitResult _outhit))))
    (:then_1
      (Utilities|SpawnDecalSimple (Math|Vector|MakeVector 5e+01 (* (ScalableFloat|GetValueatLevel (Variables|Config|GetAoESize)) 0.7) (* (ScalableFloat|GetValueatLevel (Variables|Config|GetAoESize)) 0.7)) (Math|Vector|MakeVector 5e+01 (* (ScalableFloat|GetValueatLevel (Variables|Config|GetAoESize)) 0.7) (* (ScalableFloat|GetValueatLevel (Variables|Config|GetAoESize)) 0.7)) Location (Math|Rotator|CombineRotators (Math|Rotator|MakeRotator 0.0 -9e+01 (Math|Random|RandomFloatinRange -9e+01 9e+01)) (Math|Rotator|MakeRotfromZX (LocalVariables|GetImpactNrm) (Math|Vector|MakeVector (Math|Random|RandomFloatinRange -1.0 1.0)))) (ScalableFloat|GetValueatLevel (Variables|Config|GetDecalDuration)) (ScalableFloat|GetValueatLevel (Variables|Config|GetDecalDuration)) "/Game/Weapons/Materials/MD_DecalSimple_Master_01.MD_DecalSimple_Master_01" "/Game/Effects/Fort_Effects/Gameplay/SmokeDustExplosions/Textures/T_GroundChar_01.T_GroundChar_01" "(R=0.000000,G=0.000000,B=0.000000,A=0.000000)" "(R=0.000000,G=0.000000,B=0.000000,A=0.000000)"))))

(fn ConstructionScript)
